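# Post-process AnimBP for the female AlderOptic head that adjusts head rotation relative to neck_02 when the helmet is equipped.
# A cosmetic meta tag detects the helmet on customization load.
# Each update computes scaled additive rotations from the head's delta from ref pose, normally countering part of it with yaw-dependent strength.
# During the front-end lobby idle montage with its bypass curve active, the graph blends to a head constrained to neck_02 plus re-scaled rotation.
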AnimBlueprint /BRCosmetics/Characters/Player/Female/Medium/Heads/F_MED_AlderOptic_Head/Meshes/F_MED_AlderOptic_Head_Post_AnimBP
target_skeleton=/Game/Characters/Player/Common/Fortnite_Base_Head/Fortnite_Base_3L_Head_Skeleton
graphs [
 AnimGraph {nodes [
   n1 = "Constraint - head (P)"(ComponentPose=Local To Component(LocalPose=Use cached pose 'inputCP'()), Alpha=1, ConstraintWeights_0=1) {boneToModify {boneName=head}, constraintSetup [{targetBone {boneName=neck_02}}], constraintWeights [1]}
   n2 = "Transform (Modify) Bone - Bone: head"!(ComponentPose=@n1, Alpha=1, Rotation=Get headRot_IdleMontageHelmet() -> (headRot_IdleMontageHelmet:Rotator)) -> (Pose) {rotationMode=BMM_Additive, rotationSpace=BCS_BoneSpace, alphaBoolBlend {blendOption=Linear}, boneToModify {boneName=head}}
   n3 = "Transform (Modify) Bone - Bone: head"(ComponentPose=Local To Component(LocalPose=Use cached pose 'inputCP'()), Alpha=1, Rotation=Get headRot() -> (headRot:Rotator)) {rotationMode=BMM_Additive, rotationSpace=BCS_BoneSpace, alphaBoolBlend {blendOption=Linear}, boneToModify {boneName=head}}
   Output Pose(Result=Blend Poses by bool(
     BlendPose_0=Component To Local(ComponentPose=@n2)
     BlendPose_1=Component To Local(ComponentPose=@n3)
     BlendTime_0=0.5
     BlendTime_1=0.5
     bActiveValue=Get isIdleMontageHelmet() -> (isIdleMontageHelmet:bool)
    ))
   Save cached pose 'inputCP'(Pose=Input Pose()) {cacheName=inputCP}
  ]}
 EventGraph {nodes [
   Event BlueprintUpdateAnimation()
   > n1 = Sequence()
   n2 = GetAnimInstance(self=GetOwnerSkeletalMeshComponent())
   n3 = GetCurveValue(self=@n2, CurveName=FrontEnd_Idle_Bypass)
   > Set isIdleMontageHelmet(isIdleMontageHelmet:bool=AND Boolean!(
     A="To Bool (Double)"!(Value:double=@n3) -> (ReturnValue:bool)
     B=Montage_IsPlaying(self=@n2, Montage=/BRCosmetics/Animation/Game/MainPlayer/Menu/BR/CustomSkins/_s40/AlderOptic_F/AlderOptic_Lobby_Idle_Montage)
     C:bool=Get bIsInFrontEnd()
     D:bool=Get IsHelmetEquipped() -> (IsHelmetEquipped:bool)
    ) -> (ReturnValue)) -> (Output_Get:bool)
   Branch(execute=@n1.then_1, Condition=Get IsHelmetEquipped())
   > n4 = Branch(Condition=Get isIdleMontageHelmet())
   n5 = BreakRotator(InRot=GetDeltaTransformFromRefPose!(self=GetOwnerSkeletalMeshComponent(), BoneName=head, BaseName=neck_02) -> (ReturnValue_Location:Vector, ReturnValue_Rotation:Rotator, ReturnValue_Scale:Vector).ReturnValue_Rotation)
   n6 = float * float(A=@n5.Yaw, B=Remap Float(Input=@n5.Yaw, OldRange=(X=0,Y=-10), NewRange=(X=0,Y=1), Clamp=true) -> (Output=~)) {is=title K2Node_PromotableOperator}
   > Set headRot_IdleMontageHelmet(headRot_IdleMontageHelmet:Rotator=MakeRotator(Roll=float * float(A=@n5, B=0.5), Pitch=float * float(A=@n5.Pitch, B=0.1), Yaw=float + float(A=@n6, B=-2.5))) -> (Output_Get:Rotator)
   n7 = BreakRotator(InRot=GetDeltaTransformFromRefPose(self=GetOwnerSkeletalMeshComponent(), BoneName=head, BaseName=neck_02).ReturnValue_Rotation)
   n8 = Lerp(A=0.2, B=0.8, Alpha=Remap Float(Input=@n7.Yaw, OldRange=(X=-10,Y=10), NewRange=(X=0,Y=1), Clamp=true) -> (Output=~))
   Set headRot(
    execute=@n4.else
    headRot:Rotator=MakeRotator(
     Roll=float * float!(A=@n7, B=-1, C:double=0.75) -> (ReturnValue)
     Pitch=float * float(A=@n7.Pitch, B=-1, C=0.3)
     Yaw=float * float(A=@n7.Yaw, B=-1, C=@n8)
    )
   ) -> (Output_Get:Rotator)
   Event OnCharacterCustomizationLoaded()
   > Set IsHelmetEquipped(IsHelmetEquipped:bool=CharacterItemDefinitionHasAnyMetaTag(Tags=Make Array("[0]"=(TagName="Cosmetics.Metatags.AlderOpticHelmet")))) -> (Output_Get:bool)
  ]}
]
variables [
 headRot {type=Rotator}
 headRot_IdleMontageHelmet {type=Rotator}
 IsHelmetEquipped {type=boolean}
 isIdleMontageHelmet {type=boolean}
]
